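# Masked parent material for KelpLinen characters, adding glow-colored rim lighting and a dissolve to the shared character shading.
# The Phase1 static switch picks a flat emissive gradient look or darkened base color with rim, underlight and emissive, both tinted from MPC_KelpLinen.
# Pre-skinned sphere masks localize gradient and lights, with audio-reactive beat and hand-glow masks also defined.
# A location fade drives a height-lerped, temporally dithered opacity mask.
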Material /ClydeCosmetics/Environments/KelpLinen/Materials/Kelp_Linen/M_KelpLinenParent
graph {
 properties {
  allowTranslucentCustomDepthWrites=true
  bUseMaterialAttributes=true
  bUsedWithClothing=true
  bUsedWithMeshDeformer=true
  bUsedWithSkeletalMesh=true
  blendMode=BLEND_Masked
  pixelDepthOffsetMode=PDOM_Legacy
  shadingModel=MSM_FromMaterialExpression
  usedShadingModels=Default Lit | Subsurface | Clear Coat | Subsurface Profile | Cloth
 }
 nodes [
  Add(
   A=Transform() {transformSourceType=TRANSFORMSOURCE_Local}
   B=MaterialFunctionCall() {materialFunction=&ComponentPivotLocation}.Result
  )
  n1 = Time()
  n2 = ScalarParameter() {group=Kelp, parameterName=Use_OffBeat, sortPriority=0}
  n3 = Frac(None=Add(A=Multiply(A=@n1) {constB=0.6522}, B=LinearInterpolate(Alpha=@n2) {constB=0.4253}))
  n4 = Multiply(A=Power(Base=ComponentMask(None=@n3) {r=true}) {constExponent=1}) {constB=8}
  n5 = OneMinus(None=@n3)
  n6 = Multiply(A=Power(Base=ComponentMask(None=@n5) {r=true}) {constExponent=4})
  Clamp(None=Multiply(A=Multiply(A=@n4, B=@n6)) {constB=1.4})
  Constant3Vector() {constant=<rgba -17.8759 21.5728 137.8685 1>}
  Constant() {r=47.2127}
  Constant() {r=5}
  n7 = VectorParameter() {group=Hand_Light, parameterName=CenterPoint2, sortPriority=10, defaultValue=<rgba -30 20 102.233 1>}
  Distance(A=@n7.RGB, B=NamedRerouteUsage())
  n8 = Constant()
  n9 = If(
   A=Add(A=PixelDepth().R, B=ScalarParameter() {defaultValue=-20000000, parameterName=BiasSorting_Far})
   B=ComponentMask(None=SceneTexture() {sceneTextureId=PPI_CustomDepth}.Color) {r=true}
   "A > B"=@n8
   "A == B"=@n8
   "A < B"=Constant() {r=1}
  )
  n10 = Constant() {r=1}
  FeatureLevelSwitch(Default=@n9, ES2=@n10, ES3_1=@n10, SM4_REMOVED=@n9)
  LocalPosition() {localOrigin=InstancePreSkinning}
  n11 = ScalarParameter() {defaultValue=45, group=Hand_Light, parameterName=Mask Radius B, sortPriority=25}
  Multiply(A=MaterialFunctionCall() {materialFunction=&ObjectScale}.Scale X, B=@n11)
  n12 = VectorParameter() {group=Hand_Light, parameterName=CenterPoint2, sortPriority=10, defaultValue=<rgba -30 20 102.233 1>}
  n13 = ScalarParameter() {defaultValue=0.2, group=Hand_Light, parameterName=Mask_B_Contrast}
  n14 = Saturate(None=SphereMask(A=@n12.RGB, B=NamedRerouteUsage(), Radius=@n11, Hardness=@n13))
  n15 = Normalize(VectorInput=CameraVectorWS())
  n16 = Normalize(VectorInput=NamedRerouteUsage())
  n17 = Power(
   Base=Saturate(None=DotProduct(A=@n15, B=@n16))
   Exp=ScalarParameter() {defaultValue=2, group=Hand_Light, parameterName=Fresnel Contrast B, sortPriority=36}
  )
  n18 = ScalarParameter() {defaultValue=1, group=Hand_Light, parameterName=Fresnel Brightness B, sortPriority=37}
  n19 = NamedRerouteUsage()
  n20 = DotProduct(A=Normalize(VectorInput=Subtract(A=@n12.RGB, B=@n19)), B=@n16)
  n21 = Subtract(A=@n12.RGB, B=NamedRerouteUsage())
  n22 = VertexNormalWS()
  n23 = Saturate(None=DotProduct(A=Normalize(VectorInput=@n21), B=Normalize(VectorInput=@n22)))
  n24 = StaticSwitch(True=Constant() {r=1}, False=Constant(), Value=StaticBool() {value=true})
  n25 = Add(
   A=Saturate(None=Multiply(A=@n17, B=@n18))
   B=LinearInterpolate(B=Saturate(None=@n20), Alpha=LinearInterpolate(B=@n23, Alpha=@n24) {constA=1})
  )
  n26 = VectorParameter() {group=Hand_Light, parameterName=Color B, sortPriority=20, defaultValue=<rgba 0.0592 0.9473 0.5151 1>}
  n27 = ScalarParameter() {defaultValue=4, group=Hand_Light, parameterName=light intensity B}
  Multiply(A=@n14, B=Multiply(A=Multiply(A=@n25, B=@n26.RGB) {desc=Color}, B=@n27))
  n28 = Constant() {r=1}
  Multiply(
   A=@n27
   B=MaterialFunctionCall(Value_Morning=@n28, Value_Day=@n28, Value_Evening=@n28, Value_Night=Constant() {r=0.5}) {materialFunction=&Scalar_Time_of_Day}.Result
  )
  n29 = SphereMask(A=NamedRerouteUsage(), B=@n12.RGB, Radius=@n11, Hardness=@n13)
  n30 = DotProduct(A=@n29, B=NamedRerouteUsage())
  Multiply(A=@n29, B=@n30)
  n31 = TextureCoordinate() {coordinateIndex=1}
  n32 = TextureCoordinate()
  n33 = MaterialFunctionCall!(UV=NamedRerouteDeclaration(None=StaticSwitchParameter(True=@n31, False=@n32) {group=00_Texture Inputs, parameterName=useSecondUVChannel, sortPriority=40}) {name=UVs, nodeColor=<rgba 0.2891 0 1 1>}) {materialFunction=/Game/Global/Materials/Character_Parent_2023/Functions/MF_CharacterParent_2023_TextureInputs}
  n34 = MaterialFunctionCall(Float3=@n33.M) {desc="R = AO\r\nG = GMap\r\nB = Skin", materialFunction=&BreakOutFloat3Components}
  NamedRerouteDeclaration(None=@n34.R) {name=Ambient Occlusion, nodeColor=<rgba 0.875 0 1 1>}
  NamedRerouteDeclaration(None=@n34.G) {name=GMap, nodeColor=<rgba 0.6641 0 1 1>}
  n35 = Constant() {r=8192}
  n36 = ComponentMask(None=SceneTexture() {sceneTextureId=PPI_CustomDepth}.Color) {r=true}
  n37 = Divide(A=@n35, B=PixelDepth().R)
  n38 = Constant() {r=0.0001}
  n39 = Ceil(None=Subtract(A=Divide(A=@n35, B=@n36), B=Add(A=@n37, B=@n38) {constB=0.0001}))
  NamedRerouteDeclaration(None=LinearInterpolate(A=MaterialFunctionCall() {materialFunction=/ClydeCosmetics/Environments/KelpLinen/MF_KelpLinen_LocationFade}.Output, Alpha=Clamp(None=@n39)) {constA=0.5, constB=0}) {name=Opacity, nodeColor=<rgba 1 0 0.5625 1>}
  n40 = ScalarParameter() {defaultValue=2, group=RimLight, parameterName=RimLight_Str_Hair}
  n41 = Add(A=@n40, B=ScalarParameter() {defaultValue=1, group=RimLight, parameterName=Rim_Beat_Adjust})
  n42 = NamedRerouteUsage()
  n43 = Multiply(A=StaticSwitchParameter(True=@n40, False=LinearInterpolate(A=@n40, B=@n41, Alpha=@n42)) {parameterName=Phase1}, B=NamedRerouteUsage())
  n44 = TextureSample() {
   samplerType=SAMPLERTYPE_Masks
   texture=/BRCosmetics/Characters/Player/Male/Medium/Bodies/M_MED_KelpLinenCalcium/Textures/T_KelpLinenCalcium_FaceAcc_FX
  }
  n45 = OneMinus(None=@n44.R)
  n46 = ScalarParameter() {defaultValue=2, group=RimLight, parameterName=RimLight_Str}
  n47 = ScalarParameter() {defaultValue=1, group=RimLight, parameterName=Rim_Beat_Adjust}
  n48 = StaticSwitchParameter(True=@n46, False=LinearInterpolate(A=@n46, B=Add(A=@n46, B=@n47), Alpha=NamedRerouteUsage())) {parameterName=Phase1}
  n49 = MaterialFunctionCall() {materialFunction=&MF_CharacterParent_2023_PreFX}
  n50 = GetMaterialAttributes(None=@n49.Result) {attributeGetTypes [#1, #2, #3]}
  n51 = Transform(None=@n50.Normal)
  n52 = MaterialFunctionCall!(
   BaseColor=Constant() {r=1}
   Gmap=StaticSwitchParameter(True=LinearInterpolate(A=@n43, B=@n48, Alpha=@n45), False=@n48) {parameterName=FaceAccessory?}
   Normals=@n51
   LightVector=VectorParameter() {group=RimLight, parameterName=Light_Direction, defaultValue=<rgba -0.016 0.3797 0.304 1>}.RGB
  ) {materialFunction=/ClydeCosmetics/Environments/KelpLinen/MF_KelpLinen_Rimlight}
  n53 = NamedRerouteUsage()
  n54 = MaterialFunctionCall(
   In=Fresnel(ExponentIn=ScalarParameter() {defaultValue=0.7, group=RimLight, parameterName=Fresnel Exp}, Normal=@n51) {baseReflectFraction=0}
   Contrast=Constant() {r=0.03}
  ) {materialFunction=&CheapContrast}
  n55 = Multiply(A=StaticSwitchParameter(True=@n52.Result, False=Multiply(A=@n52.Result, B=@n53)) {parameterName=FaceAccessory?}, B=@n54.Result)
  n56 = Add(A=@n55, B=Multiply(A=Multiply(A=@n54.Result) {constB=0.25}, B=NamedRerouteUsage()))
  n57 = MaterialFunctionCall(
   BaseColor=Constant() {r=1}
   Gmap=ScalarParameter() {defaultValue=1.1398, parameterName=RimLight_Underlight}
   Normals=@n51
   LightVector=VectorParameter() {parameterName=Light_Direction_5, defaultValue=<rgba 0 -1 -0.3 1>}.RGB
  ) {materialFunction=&MF_KelpLinen_Rimlight}
  n58 = ScalarParameter() {defaultValue=0.5, parameterName=Underlight_Str}
  n59 = Multiply(
   A=SphereMask(
    A=NamedRerouteUsage()
    B=VectorParameter() {group=Kelp Underlight, parameterName=Underlight_Skeletal_Mask, defaultValue=<rgba 0 0 60 1>}.RGB
    Radius=ScalarParameter() {defaultValue=120, group=Kelp Underlight, parameterName=Light_Gradient_Radius}
    Hardness=ScalarParameter() {defaultValue=0.2, group=Kelp Underlight, parameterName=Light_Softness}
   )
   B=DotProduct(A=Multiply(A=@n57.Result, B=@n58), B=NamedRerouteUsage())
  ) {constA=1}
  n60 = Add(A=StaticSwitchParameter(True=Multiply(A=@n59, B=@n45), False=@n59) {parameterName=FaceAccessory?}, B=@n55)
  NamedRerouteDeclaration(None=Saturate(None=StaticSwitchParameter(True=@n56, False=@n60) {parameterName=Phase1})) {name=Fresnel MAsk, nodeColor=<rgba 1 0 0.5156 1>}
  NamedRerouteDeclaration(None=@n50.Normal) {name=Normal, nodeColor=<rgba 1 0.9844 0 1>}
  n61 = VectorParameter() {group=Hand_Light, parameterName=Color B, sortPriority=20, defaultValue=<rgba 0.0592 0.9473 0.5151 1>}
  n62 = Multiply(A=@n61.RGB, B=ScalarParameter() {defaultValue=4, group=Hand_Light, parameterName=light intensity B})
  n63 = Constant() {r=1}
  n64 = MaterialFunctionCall(Value_Morning=@n63, Value_Day=@n63, Value_Evening=@n63, Value_Night=Constant() {r=0.5}) {materialFunction=&Scalar_Time_of_Day}
  NamedRerouteDeclaration(None=Multiply(A=Multiply(A=@n62, B=@n64.Result), B=@n30)) {name=Hand_Glow, nodeColor=<rgba 0.5 0 1 1>}
  NamedRerouteDeclaration(None=@n49.Pre skinned Local Position) {name=Local Position, nodeColor=<rgba 0 0.7891 1 1>}
  NamedRerouteDeclaration(None=@n30) {name=hand_Glow_Mask, nodeColor=<rgba 0 0.9531 1 1>}
  n65 = MaterialFunctionCall() {materialFunction=&MF_KelpLinen_LocationFade}
  n66 = NamedRerouteUsage()
  n67 = VertexInterpolator(VS=SphereMask(
    A=MaterialFunctionCall(Float3=@n66) {materialFunction=&BreakOutFloat3Components}.B
    B=ScalarParameter() {defaultValue=60, group=Kelp Opacity, parameterName=Dissolve Z offset}
    Radius=ScalarParameter() {defaultValue=119.04, group=Kelp Opacity, parameterName=dissolve Radius}
    Hardness=ScalarParameter() {defaultValue=0.2, group=Kelp Opacity, parameterName=Dissolve Softness}
   ))
  NamedRerouteDeclaration(None=MaterialFunctionCall(Alpha Threshold=MaterialFunctionCall(Transition Phase=@n65.Output, Height Texture=@n67.PS) {materialFunction=&HeightLerp}.Alpha) {materialFunction=&DitherTemporalAA}.Result) {name=Opacity Mask, nodeColor=<rgba 1 0.7734 0 1>}
  n68 = MaterialFunctionCall(Float3=NamedRerouteUsage()) {materialFunction=&BreakOutFloat3Components}
  n69 = ScalarParameter() {defaultValue=125.3845, group=Kelp Gradient, parameterName=Gradient Z offset}
  n70 = ScalarParameter() {defaultValue=30.3053, group=Kelp Gradient, parameterName=Gradient Radius}
  n71 = ScalarParameter() {defaultValue=0.1, group=Kelp Gradient, parameterName=Gradient Softness}
  NamedRerouteDeclaration(None=VertexInterpolator(VS=SphereMask(A=@n68.B, B=@n69, Radius=@n70, Hardness=@n71)).PS) {name=Mannequin Gradient, nodeColor=<rgba 1 0.3047 0 1>}
  NamedRerouteDeclaration(None=MaterialFunctionCall!(In) {materialFunction=/ClydeCosmetics/Environments/KelpLinen/MF_Kelp_AudioReactive}.Beat_Tracker) {name=Music Beat, nodeColor=<rgba 0.1484 1 0 1>}
  NamedRerouteDeclaration(None=@n50.Base Color) {name=base color, nodeColor=<rgba 0 1 0.0391 1>}
  NamedRerouteDeclaration(None=@n34.B) {name=Skin, nodeColor=<rgba 0.6641 0 1 1>}
  n72 = LocalPosition() {localOrigin=InstancePreSkinning}
  NamedRerouteDeclaration(None=SphereMask(
    A=VertexInterpolator(VS=@n72.XYZ).PS
    B=VectorParameter() {group=ColorFade, parameterName=Color_Fade_Position, defaultValue=<rgba 0 4 167 1>}.RGB
    Radius=ScalarParameter() {group=ColorFade, parameterName=Color_Fade_Radius}
    Hardness=ScalarParameter() {defaultValue=1, group=ColorFade, parameterName=Color_contrast}
   )) {name=Color Mask, nodeColor=<rgba 0.8047 1 0 1>}
  n73 = PreSkinnedPosition()
  NamedRerouteDeclaration(None=ShaderStageSwitch(PixelShader=VertexInterpolator(VS=@n73).PS, VertexShader=@n73)) {name=Pre Skinned Local Postition, nodeColor=<rgba 1 0.5156 0 1>}
  NamedRerouteDeclaration(None=GetMaterialAttributes(None=@n33.Result) {attributeGetTypes [#1]}.Normal) {name=Normal Map, nodeColor=<rgba 1 0.3281 0 1>}
  NamedRerouteDeclaration(None=Fresnel(ExponentIn=Constant() {r=1}) {baseReflectFraction=0}) {name=Fresnel, nodeColor=<rgba 1 0.3047 0 1>}
  NamedRerouteDeclaration(None=@n33.M) {name=Mask, nodeColor=<rgba 0.3125 0 1 1>}
  n74 = TextureSampleParameter2D() {
   group=05_Tech Art Masks
   parameterName=TechArtMask
   samplerType=SAMPLERTYPE_Masks
   sortPriority=0
   texture=&T_BlackMask
  }
  n75 = StaticSwitchParameter(True=@n74.RGB, False=Constant()) {defaultValue=true, group=09_DisableTextureSamplers, parameterName=useTechArtMask}
  n76 = ScalarParameter() {desc=Set to 1 to indicate to use TechArtMask for opacity in Hologram, group=05_Tech Art Masks, parameterName=UseTechArtMaskForOpacity}
  NamedRerouteDeclaration(None=LinearInterpolate(A=@n75, B=@n75, Alpha=@n76)) {name=Tech Art Mask, nodeColor=<rgba 1 0.7031 0 1>}
  n77 = VertexInterpolator(VS=LocalPosition() {localOrigin=InstancePreSkinning}.XYZ)
  n78 = VectorParameter() {group=RimLight, parameterName=Light_Fade_Position, defaultValue=<rgba 0 4 167 1>}
  n79 = ScalarParameter() {defaultValue=21.5, group=RimLight, parameterName=Light_Fade_Radius}
  n80 = ScalarParameter() {defaultValue=1, group=RimLight, parameterName=Light_contrast}
  NamedRerouteDeclaration(None=OneMinus(None=SphereMask(A=@n77.PS, B=@n78.RGB, Radius=@n79, Hardness=@n80))) {name=Fade_Mask, nodeColor=<rgba 0.8281 1 0 1>}
  NamedRerouteUsage()
  Normalize(VectorInput=Subtract(A=WorldPosition().XYZ, B=VectorParameter() {parameterName=Test Light, defaultValue=<rgba 0 0 0 1>}.RGB))
  PixelNormalWS()
  PreSkinnedPosition()
  Saturate(None=Multiply(A=Multiply(A=@n14, B=@n25), B=@n27))
  n81 = MaterialFunctionCall!(Skin Mask=NamedRerouteUsage(), GMap=NamedRerouteUsage(), Material=@n33.Result) {materialFunction=/Game/Global/Materials/Character_Parent_2023/Functions/MF_CharacterParent_2023_Skin}
  n82 = MaterialFunctionCall!(
   Material=MaterialFunctionCall!(
    Material=MaterialFunctionCall!(Material=@n81.Result) {materialFunction=/Game/Packages/Fortress_SharedMaterials/Base_Material_Functions/CharacterBrightnessandScale_Balance}.Result
    Emissive UVs
    Fresnel=NamedRerouteUsage()
    Alpha
   ) {materialFunction=/Game/Global/Materials/Character_Parent_2023/Functions/MF_CharacterParent_2023_Emissive}.Result
   Ao/Gmap/Skin
   Fresnel=NamedRerouteUsage()
   Alpha
  ) {materialFunction=/Game/Global/MaterialFunctions/TechArt/MF_ClothFuzz_TechArt}
  n83 = MaterialFunctionCall!(
   Emissive
   Alpha=ScalarParameter() {defaultValue=1, group=Cosmetic Emissive EyeAdaptation, parameterName=Cosmetic Emissive EyeAdaptation Alpha}
   Material=@n82.Result
  ) {materialFunction=/Game/Packages/Fortress_SharedMaterials/Base_Material_Functions/BaseCharacterShaderFunctions/MF_CosmeticEmissiveEyeAdaptation}
  StaticSwitchParameter(True=@n83.Material, False=@n82.Result) {defaultValue=true, group=Cosmetic Emissive EyeAdaptation, parameterName=Cosmetic Emissive EyeAdaptation, sortPriority=100}
  n84 = CollectionParameter() {collection=/ClydeCosmetics/Environments/KelpLinen/Materials/MPC_KelpLinen, parameterName=KelpLinen_Glow_Color}
  n85 = ScalarParameter() {defaultValue=0.7, group=Kelp Gradient, parameterName=Gradient_Multi}
  n86 = Multiply(A=NamedRerouteUsage(), B=ScalarParameter() {parameterName=Phase_1_Basecolor_Str})
  n87 = Constant()
  n88 = Add(
   A=LinearInterpolate(A=Multiply(A=@n84, B=@n85), B=@n84, Alpha=NamedRerouteUsage())
   B=Multiply(A=AppendVector(A=@n86, B=@n87), B=NamedRerouteUsage())
  )
  n89 = ScalarParameter() {defaultValue=1.5, group=Kelp Gradient, parameterName=Rim_Adjustment}
  n90 = LinearInterpolate(A=@n89, B=Add(A=@n89, B=ScalarParameter() {defaultValue=0.5, group=RimLight, parameterName=Rim_Beat_Adjust2}), Alpha=NamedRerouteUsage())
  n91 = NamedRerouteUsage()
  n92 = SetMaterialAttributes!(
   MaterialAttributes=@n50.MaterialAttributes
   Base Color
   Normal=Constant3Vector()
   Roughness=Constant() {r=1}
   Specular=Constant() {r=0.2}
   Emissive Color=LinearInterpolate(A=@n88, B=Multiply(A=@n88, B=@n90) {constB=0.25}, Alpha=@n91)
   Opacity Mask=NamedRerouteUsage()
  ) {attributeSetTypes [
    #2
    #1
    #4
    #5
    #3
    #6
   ]}
  n93 = ScalarParameter() {parameterName=BaseColor_Influence, primitiveDataIndex=11}
  n94 = Multiply(A=@n84, B=@n91)
  n95 = Add(A=Add(A=Multiply(A=@n50.Base Color, B=@n93), B=ComponentMask(None=@n94) {b=true, g=true, r=true}), B=@n50.Emissive Color)
  n96 = NamedRerouteUsage()
  n97 = NamedRerouteUsage()
  n98 = ScalarParameter() {defaultValue=0.2, parameterName=Base Color Multi}
  StaticSwitchParameter(
   True=@n92
   False=SetMaterialAttributes!(
    MaterialAttributes=@n50.MaterialAttributes
    Emissive Color=LinearInterpolate(A=@n95, B=@n96, Alpha=@n97)
    Base Color=Multiply(A=@n50.Base Color, B=@n98)
    Opacity Mask=NamedRerouteUsage()
   ) {attributeSetTypes [#3, #2, #6]}
  ) {parameterName=Phase1}
  TextureCoordinate()
  VectorParameter() {parameterName=Fog_Color, defaultValue=<rgba 0.3005 0.5333 1 1>}
  VectorParameter() {group=Kelp Underlight, parameterName=Underlight_Skeletal_Mask, defaultValue=<rgba 0 0 60 1>}
  WorldPosition()
  WorldPosition()
  WorldPosition() {worldPositionShaderOffset=WPT_ExcludeAllShaderOffsets}
 ]
}
